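# Material function that layers a tintable rice paper surface onto incoming Material Attributes for the Dojo set, toggled by the UseRicePaper static switch.
# Paper texture and normal are each sampled at two tilings, with a low-res noise blend mask available to break up repetition.
# A channel-selectable mask and lerps combine paper color, roughness, opacity and normal with the input.
# The paper branch sets Two Sided Foliage shading with tinted subsurface color for translucency.
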MaterialFunction /Game/Environments/Asteria/Sets/Dojo/Materials/RicepaperMaterial/MF_Ricepaper
graph {nodes [
  n1 = FunctionInput() {inputName=MA_In, inputType=FunctionInput_MaterialAttributes}
  n2 = GetMaterialAttributes(None=@n1) {attributeGetTypes [
    #1
    #2
    #3
    #4
    #5
    #6
    #7
   ]}
  n3 = NamedRerouteUsage()
  n4 = NamedRerouteUsage()
  n5 = VectorParameter() {group=RicePaper, parameterName=RicePaperColor, defaultValue=<rgba 0.8177 0.63 0.4906 1>}
  n6 = Multiply(A=@n5.RGB, B=LinearInterpolate(Alpha=NamedRerouteUsage()) {constA=0.6})
  n7 = LinearInterpolate(Alpha=NamedRerouteUsage()) {constA=0.5}
  n8 = NamedRerouteUsage()
  n9 = ScalarParameter() {defaultValue=0.5, group=RicePaper, parameterName=Roughness MIN}
  n10 = ScalarParameter() {defaultValue=1, group=RicePaper, parameterName=Roughness MAX}
  n11 = NamedRerouteUsage()
  n12 = SetMaterialAttributes!(
   MaterialAttributes=@n2.MaterialAttributes
   EmissiveColor=@n3
   BaseColor=LinearInterpolate(A=NamedRerouteUsage(), B=LinearInterpolate(A=@n4, B=@n6) {constAlpha=0.8}, Alpha=NamedRerouteUsage())
   Roughness=LinearInterpolate(A=NamedRerouteUsage(), B=LinearInterpolate(A=@n9, B=@n10, Alpha=@n11) {constA=0.3, constB=0.5}, Alpha=NamedRerouteUsage())
   Opacity=LinearInterpolate(A=NamedRerouteUsage(), B=LinearInterpolate(Alpha=@n8) {constA=0.2, constB=0.02}, Alpha=NamedRerouteUsage()) {constA=1}
   Normal=LinearInterpolate(A=NamedRerouteUsage(), B=NamedRerouteUsage(), Alpha=NamedRerouteUsage())
   SubsurfaceColor=LinearInterpolate(B=Multiply(A=@n5.RGB, B=@n7), Alpha=NamedRerouteUsage())
   ShadingModel=ShadingModel() {shadingModel=MSM_TwoSidedFoliage}
   Specular=LinearInterpolate(A=NamedRerouteUsage(), Alpha=NamedRerouteUsage()) {constB=0.5}
   Metallic=LinearInterpolate(A=NamedRerouteUsage(), Alpha=NamedRerouteUsage()) {constB=0}
  ) {attributeSetTypes [
    #1
    #2
    #3
    #4
    #5
    #8
    #9
    #6
    #7
   ]}
  FunctionOutput(None=StaticSwitchParameter(True=@n12, False=@n1) {group=RicePaper, parameterName=UseRicePaper})
  LinearInterpolate(A=@n3, B=NamedRerouteUsage(), Alpha=NamedRerouteUsage())
  Multiply(A=@n4, B=@n6)
  n13 = TextureSample(UVs=Multiply(A=TextureCoordinate()) {constB=3}) {samplerSource=SSM_Wrap_WorldGroupSettings, samplerType=SAMPLERTYPE_Grayscale, texture=&LowResBlurredNoise}
  NamedRerouteDeclaration(None=Power(Base=@n13.R)) {name=Blend Mask, nodeColor=<rgba 0 1 0.625 1>}
  n14 = TextureSampleParameter2D(UVs=Multiply(A=TextureCoordinate(), B=Constant() {r=7}) {constB=5}) {
   group=RicePaper
   parameterName=RicePaperNM
   samplerSource=SSM_Wrap_WorldGroupSettings
   samplerType=SAMPLERTYPE_Normal
   texture=&FlatNormalBlue
  }
  n15 = TextureSampleParameter2D(UVs=Multiply(A=TextureCoordinate(), B=Constant() {r=12}) {constB=5}) {
   group=RicePaper
   parameterName=RicePaperNM
   samplerSource=SSM_Wrap_WorldGroupSettings
   samplerType=SAMPLERTYPE_Normal
   texture=&FlatNormalBlue
  }
  n16 = NamedRerouteUsage()
  NamedRerouteDeclaration(None=MaterialFunctionCall(Normal=LinearInterpolate(A=@n14.RGB, B=@n15.RGB, Alpha=@n16), Flatness=ScalarParameter() {defaultValue=1}) {materialFunction=&FlattenNormal}.Result) {name=Normal, nodeColor=<rgba 1 0 0.7266 1>}
  NamedRerouteDeclaration(None=@n2.BaseColor) {name=BaseColor, nodeColor=<rgba 0.2422 0 1 1>}
  NamedRerouteDeclaration(None=@n2.EmissiveColor) {name=EmissiveColor, nodeColor=<rgba 0 0.25 1 1>}
  n17 = TextureSampleParameter2D(UVs=Multiply(A=TextureCoordinate(), B=Constant() {r=7}) {constB=5}) {
   group=RicePaper
   parameterName=RicePaperTexture
   samplerSource=SSM_Wrap_WorldGroupSettings
   samplerType=SAMPLERTYPE_Masks
   texture=/Game/Global/Textures/Default/Blanks/T_White
  }
  n18 = TextureSampleParameter2D(UVs=Multiply(A=TextureCoordinate(), B=Constant() {r=12}) {constB=5}) {
   group=RicePaper
   parameterName=RicePaperTexture
   samplerSource=SSM_Wrap_WorldGroupSettings
   samplerType=SAMPLERTYPE_Masks
   texture=&T_White
  }
  NamedRerouteDeclaration(None=LinearInterpolate(A=@n17.R, B=@n18.R, Alpha=NamedRerouteUsage())) {name=PaperTexture, nodeColor=<rgba 0 1 0.25 1>}
  n19 = TextureSampleParameter2D() {group=RicePaper, parameterName=RicePaperMask, samplerType=SAMPLERTYPE_Masks, texture=&T_White}
  NamedRerouteDeclaration(None=StaticComponentMaskParameter(None=@n19.RGBA) {defaultR=true, group=RicePaper, parameterName=RicePaperMaskChannnel}) {name=mask, nodeColor=<rgba 0.8516 0 1 1>}
  n20 = CameraVectorWS()
  n21 = PixelNormalWS()
  n22 = LinearInterpolate(Alpha=Power(Base=DotProduct(A=@n20, B=@n21)) {constExponent=20}) {constB=0.1}
  n23 = NamedRerouteUsage()
  NamedRerouteDeclaration(None=Multiply(
    A=Multiply(A=@n22, B=@n23)
    B=MaterialFunctionCall() {materialFunction=&Color_Time_of_Day}.Result
   )) {name=Emissive, nodeColor=<rgba 1 0 0.9375 1>}
  NamedRerouteDeclaration(None=@n2.Normal) {name=material Normal, nodeColor=<rgba 0 1 0.4375 1>}
  NamedRerouteDeclaration(None=@n2.Specular) {name=Specular, nodeColor=<rgba 0 1 0.3203 1>}
  NamedRerouteDeclaration(None=@n2.Metallic) {name=Metallic, nodeColor=<rgba 1 0.9844 0 1>}
  NamedRerouteDeclaration(None=@n2.Opacity) {name=Opacity, nodeColor=<rgba 1 0.0234 0 1>}
  NamedRerouteDeclaration(None=@n2.Roughness) {name=Roughness, nodeColor=<rgba 0 0.3437 1 1>}
 ]}
