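# Simple rice paper master material for the Dojo set in the Asteria environment, flagged for use with Nanite.
# It uses the Two Sided Foliage shading model, which supports light transmission through thin surfaces.
# It has no textures, only scalar and vector parameters for base color, metallic, specular, roughness, and subsurface color and opacity, with red-tinted default colors.
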Material /Game/Environments/Asteria/Sets/Dojo/Materials/RicepaperMaterial/M_RicePaper_Simple
graph {
 properties {bUsedWithNanite=true, pixelDepthOffsetMode=PDOM_Legacy, shadingModel=MSM_TwoSidedFoliage, usedShadingModels=Two Sided Foliage}
 nodes [
  ScalarParameter() {defaultValue=0.7, parameterName=Metallic}
  ScalarParameter() {defaultValue=0.3, parameterName=Specular}
  ScalarParameter() {defaultValue=0.05, parameterName=Roughness}
  ScalarParameter() {parameterName=SSSOpacity}
  VectorParameter() {parameterName=BaseColor, defaultValue=<rgba 0.224 0.0128 0.0128 0>}
  VectorParameter() {parameterName=SSSColor, defaultValue=<rgba 1 0 0 0>}
 ]
}
